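; Event-triggered gameplay ability for the IronLog ZeroPoint Dash exotic that applies configured effects (noted as IcyFeet) to hostile pawns.
; On activation it caches the avatar pawn and ability system component, creates a damage-dealt event wait task, and clears bCanApplyEffect.
; The hostile check applies effects independently of that flag.
; A short timer then restores the flag; despite the helper names, no array reset or ability end is shown.
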(event Ability|EventActivateAbilityFromEvent
  (params (EventData :type Gameplay%20Event%20Data%20Structure%20%28by%20ref%29))
  (bind (_eventtag _instigator _target) (Utilities|Struct|BreakGameplayEventData EventData))
  (bind (_asfort_pawn _bsuccess) (Utilities|Casting|CastToFortPawn _target))
  (Utilities|FlowControl|Sequence
    (:then_0
      (bind _outfortplayerpawn (CallFunction|GetAvatarasValidFortPlayerPawn)
        (:ValidFortPlayerPawn
          (bind _output_get (Variables|Default|SetFortPlayerPawn _outfortplayerpawn))
          (bind _output_get_2 (Variables|Default|SetOwnerFortAbilitySystemComponent (Ability|GetFortAbilitySystemComponent _output_get)))
          (Utilities|IsValid _output_get_2
            (:Is%20Valid
              (bind _asynctaskproxy (Ability|Tasks|WaitGameplayEvent "(TagName=\"Event.Damage.Dealt\")"))
              (Variables|Default|SetOnDamageDealtAsyncTask _asynctaskproxy)
              (if (Variables|Default|GetbCanApplyEffect)
                (Variables|Default|SetbCanApplyEffect false)))))))
    (:then_1
      ;[1 1 1] Apply IcyFeet to target
      (if (and _bsuccess (Utilities|Enum|Equal(Enum) (Team|GetTeamAffiliation _instigator _target) "Hostile"))
        (Fortnite|ApplyGameplayEffectsToActor _instigator (Variables|Default|GetEffects%20to%20Apply) (Variables|Default|GetSource%20Pos) false 1)
        (else
          _))
      (CallFunction|SetTimerBeforeResetandAbilityEnd))))

(event Custom|SetTimerBeforeResetAndAbilityEnd
  ;[1 1 1] Reset array
  (Utilities|Time|SetTimerbyEvent (EventDispatchers|CreateEvent :!event Custom|ResetHitTargetArray) 0.1 false true))

(event Custom|ResetHitTargetArray
  ;[1 1 1] Reset array
  (Variables|Default|SetbCanApplyEffect true))
